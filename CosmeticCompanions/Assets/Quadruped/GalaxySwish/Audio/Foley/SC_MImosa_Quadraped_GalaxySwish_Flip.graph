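# Sound cue for the flip foley of the GalaxySwish quadruped cosmetic companion.
# A Random node picks one of three flip wave variations with equal weight, played louder and lower in pitch.
# It uses companion foley attenuation, the creature pets sound class, and shares the quadruped footstep concurrency group, so flips compete with footsteps for voices.
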SoundCue /CosmeticCompanions/Assets/Quadruped/GalaxySwish/Audio/Foley/SC_MImosa_Quadraped_GalaxySwish_Flip
root=Random_0
nodes [
 {id=Random_0, children [WavePlayer_0,WavePlayer_1,WavePlayer_2], properties {weights [1,1,1]}}
 {
  id=WavePlayer_0
  properties {soundWaveAssetPtr=/CosmeticCompanions/Assets/Quadruped/GalaxySwish/Audio/Foley/Wavs/Mimosa_Quadruped_GalaxySwish_Flip_03}
 }
 {
  id=WavePlayer_1
  properties {soundWaveAssetPtr=/CosmeticCompanions/Assets/Quadruped/GalaxySwish/Audio/Foley/Wavs/Mimosa_Quadruped_GalaxySwish_Flip_02}
 }
 {
  id=WavePlayer_2
  properties {soundWaveAssetPtr=/CosmeticCompanions/Assets/Quadruped/GalaxySwish/Audio/Foley/Wavs/Mimosa_Quadruped_GalaxySwish_Flip_01}
 }
]
properties {
 attenuationSettings=/Game/Sounds/Attenuation/AS_Companion_Foley
 maxDistance=1200
 pitchMultiplier=0.8
 priority=0
 soundClassObject=&Fort_Creatures_Pets
 volumeMultiplier=2.5
 concurrencySet [/CosmeticCompanions/Assets/Audio_Shared/Foley/Quadraped/CONC_Mimosa_Quadraped_Footstep]
}
