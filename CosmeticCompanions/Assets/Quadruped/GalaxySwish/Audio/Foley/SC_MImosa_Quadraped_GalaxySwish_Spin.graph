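# Sound cue for the spin foley of the GalaxySwish quadruped cosmetic companion.
# A Random node picks one of three equally weighted Spin wave variations.
# Output is boosted in volume, routed to the pet creature sound class and spatialized with the shared companion foley attenuation at zero priority.
# It shares the quadruped footstep concurrency set, so spins are limited under the same rules as footsteps.
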SoundCue /CosmeticCompanions/Assets/Quadruped/GalaxySwish/Audio/Foley/SC_MImosa_Quadraped_GalaxySwish_Spin
root=Random_1
nodes [
 {id=Random_1, children [WavePlayer_3,WavePlayer_4,WavePlayer_5], properties {weights [1,1,1]}}
 {
  id=WavePlayer_3
  properties {soundWaveAssetPtr=/CosmeticCompanions/Assets/Quadruped/GalaxySwish/Audio/Foley/Wavs/Mimosa_Quadruped_GalaxySwish_Spin_01}
 }
 {
  id=WavePlayer_4
  properties {soundWaveAssetPtr=/CosmeticCompanions/Assets/Quadruped/GalaxySwish/Audio/Foley/Wavs/Mimosa_Quadruped_GalaxySwish_Spin_02}
 }
 {
  id=WavePlayer_5
  properties {soundWaveAssetPtr=/CosmeticCompanions/Assets/Quadruped/GalaxySwish/Audio/Foley/Wavs/Mimosa_Quadruped_GalaxySwish_Spin_03}
 }
]
properties {
 attenuationSettings=/Game/Sounds/Attenuation/AS_Companion_Foley
 maxDistance=1200
 priority=0
 soundClassObject=&Fort_Creatures_Pets
 volumeMultiplier=2.5
 concurrencySet [/CosmeticCompanions/Assets/Audio_Shared/Foley/Quadraped/CONC_Mimosa_Quadraped_Footstep]
}
